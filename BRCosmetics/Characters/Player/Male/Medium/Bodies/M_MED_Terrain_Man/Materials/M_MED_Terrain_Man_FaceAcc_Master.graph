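# Master material for Terrain Man's face accessory on the skinned character, drawing an animated, tinted CRT-style screen.
# PreFX output passes optional panning emissive and clear coat, then emissive fresnel, before screen emissive is set and fed into PostFX.
# The screen combines frame vignette, scanlines, static and a sweep line, which also distort the face texture UVs, masked by glitched SkinFX green plus blue.
# Map icon nodes are unconnected, and UseGlitch has identical inputs.
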Material /BRCosmetics/Characters/Player/Male/Medium/Bodies/M_MED_Terrain_Man/Materials/M_MED_Terrain_Man_FaceAcc_Master
graph {
 properties {
  bUseMaterialAttributes=true
  bUsedWithMeshDeformer=true
  bUsedWithSkeletalMesh=true
  pixelDepthOffsetMode=PDOM_Legacy
  shadingModel=MSM_Subsurface
  subsurfaceProfile=&SS_HeroSkin_02
  usedShadingModels=Subsurface
 }
 nodes [
  n1 = MaterialFunctionCall() {materialFunction=&BaseFN_CharacterFunction_PreFX}
  n2 = MaterialFunctionCall!(In=TextureObjectParameter() {group=Skin FX, parameterName=SkinFX_Mask}) {materialFunction=/Game/Weapons/WeaponSkins/MaterialFunction/MF_GlitchSubfunction}
  n3 = StaticSwitchParameter(True=@n2.Glitched, False=@n2.Glitched) {group=Skin FX, parameterName=UseGlitch}
  n4 = StaticComponentMaskParameter(None=@n3) {defaultR=true, group=Skin FX, parameterName=Clear Coat Channel}
  n5 = StaticSwitchParameter(
   True=MaterialFunctionCall!(
    Alpha=StaticComponentMaskParameter(None=@n3) {defaultR=true, group=Skin FX, parameterName=PanningEmissive_Channel}
    isVehicle
    Gmap=@n1.Gmap
    UVOffset=StaticComponentMaskParameter(None=@n3) {defaultB=true, group=Skin FX, parameterName=PanningEmissive_UVOffset_Channel}
    Local Normal
    Local Position
    EmissiveColorOverride
    Material=@n1.Gmap
   ) {materialFunction=/Game/Weapons/WeaponSkins/MaterialFunction/PanningEmissive}.Result
   False=@n1.Result
  ) {group=Skin FX, parameterName=Use Panning Emissive}
  n6 = MaterialFunctionCall!(In=StaticSwitchParameter(True=MaterialFunctionCall!(Alpha=@n4, In=@n5) {materialFunction=/Game/Weapons/WeaponSkins/MaterialFunction/Clearcoat}.Result, False=@n5) {group=Skin FX, parameterName=Use Clear Coat}) {materialFunction=/Game/Weapons/WeaponSkins/MaterialFunction/MF_EmissiveFresnel}
  n7 = Add(A=ComponentMask(None=@n3) {g=true}, B=ComponentMask(None=@n3) {b=true})
  n8 = Multiply(A=NamedRerouteUsage()) {constB=0.5}
  n9 = Multiply(A=Sine(None=ComponentMask(None=@n8) {r=true}), B=Sine(None=ComponentMask(None=@n8) {g=true}))
  n10 = OneMinus(None=Power(Base=OneMinus(None=@n9), Exp=ScalarParameter() {defaultValue=0.8, group=Background, parameterName=Frame Exponent}))
  n11 = NamedRerouteUsage()
  n12 = ConstantBiasScale(None=Frac(None=Panner(Coordinate=@n11) {speedY=-1})) {bias=-0.5, scale=2}
  n13 = Multiply(A=@n12, B=@n12)
  n14 = Add(A=Multiply(A=ComponentMask(None=@n13) {g=true}, B=ScalarParameter() {defaultValue=0.5, group=Background, parameterName=Scanline Strength}) {constB=0.1})
  n15 = NamedRerouteUsage()
  n16 = Panner(Coordinate=@n15) {speedX=5, speedY=0.1}
  n17 = Multiply(
   A=TextureSample(UVs=Panner(Coordinate=@n15) {speedX=6, speedY=-0.05}) {samplerType=SAMPLERTYPE_Masks, texture=/Game/Missions/Primary/RepairTheShelter/Textures/Cinematic_Screen_Static}.G
   B=TextureSample(UVs=Panner(Coordinate=@n15) {speedX=-3, speedY=-0.03}) {samplerType=SAMPLERTYPE_Masks, texture=&Cinematic_Screen_Static}.R
  )
  n18 = Add(
   A=Multiply(A=TextureSample(UVs=@n16) {samplerType=SAMPLERTYPE_Masks, texture=&Cinematic_Screen_Static}.B) {constB=0.1}
   B=@n17
  )
  n19 = ScalarParameter() {defaultValue=6, group=Static, parameterName=Static Scale}
  n20 = Multiply(A=ComponentMask(None=NamedRerouteUsage()) {g=true}) {constB=0.75}
  n21 = Multiply(A=Time()) {constB=-0.2}
  n22 = Power(Base=Frac(None=Add(A=@n20, B=@n21))) {constExponent=20}
  n23 = ScalarParameter() {defaultValue=0.2, group=Sweep Line, parameterName=Sweep Add}
  n24 = Add(
   A=Add(A=Multiply(A=@n18, B=@n19), B=Multiply(A=@n22, B=@n23) {constB=0.2})
   B=ScalarParameter() {defaultValue=0.2, group=Background, parameterName=Background}
  )
  n25 = LinearInterpolate(
   A=MaterialFunctionCall!(ScanLine_Tiling_1, Scanline_PannerSpeed_1, ScanLine_Tiling, Scanline_PannerSpeed, UVs) {materialFunction=/Game/Weapons/WeaponSkins/MaterialFunction/ScanLines}.Result
   B=Constant() {r=0.5}
   Alpha=MaterialFunctionCall!(Blend Range=ScalarParameter() {defaultValue=25, parameterName=r}, Start Offset=ScalarParameter() {defaultValue=-100, parameterName=d}) {materialFunction=/Engine/Functions/Engine_MaterialFunctions03/Particles/Distance_Blend}.Result
  )
  n26 = ScalarParameter() {defaultValue=0.07, group=Static, parameterName=Static UV Distort}
  n27 = ScalarParameter() {defaultValue=0.005, group=Sweep Line, parameterName=Sweep UV Distort}
  n28 = Add(A=AppendVector(A=Multiply(A=@n17, B=@n26), B=Constant()), B=AppendVector(A=Multiply(A=@n22, B=@n27) {constB=0.02}, B=Constant()))
  n29 = TextureSample(UVs=Add(A=@n28, B=TextureCoordinate())) {
   samplerType=SAMPLERTYPE_Masks
   texture=/BRCosmetics/Characters/Player/Male/Medium/Bodies/M_MED_Terrain_Man/Textures/T_M_MED_Terrain_Man_FaceAcc_E
  }
  n30 = Add(A=Multiply(A=Multiply(A=@n25, B=@n29.RGB), B=ScalarParameter() {defaultValue=5, parameterName=FaceIntensity})) {constB=0}
  n31 = Multiply(
   A=VectorParameter() {group=Colors, parameterName=Screen Tone, defaultValue=<rgba 0.0044 0.125 0.0345 0>}.RGB
   B=Multiply(A=Multiply(A=@n10, B=@n14), B=Add(A=@n24, B=@n30)) {constA=1}
  )
  MaterialFunctionCall(
   AO=@n1.AO
   Distance Blend=@n1.AO
   Fresnel=@n1.AO
   Gmap=@n1.AO
   Pre Skinned Local Position=@n1.AO
   Material=SetMaterialAttributes(
    MaterialAttributes=GetMaterialAttributes(None=@n6.Result) {attributeGetTypes [#1]}.MaterialAttributes
    Emissive Color=Multiply(A=@n7, B=@n31) {constA=1}
   ) {attributeSetTypes [#1]}
  ) {materialFunction=&BaseFN_CharacterFunction_PostFX}
  MaterialFunctionCall(f3=StaticComponentMaskParameter(None=@n3) {defaultB=true, defaultG=true, group=Skin FX, parameterName=TerrainMan Map Channel}) {materialFunction=&AddComponents}
  n32 = NamedRerouteUsage()
  MaterialFunctionCall(UVs=@n32, Texture Scale=ComponentMask(None=VectorParameter() {parameterName=MapScale, defaultValue=<rgba 1 1 0 1>}.RGB) {g=true, r=true}) {materialFunction=&ScaleUVsByCenter}
  n33 = Multiply(A=@n32, B=ComponentMask(None=VectorParameter() {parameterName=Map2Scale, defaultValue=<rgba 1.2 1.2 0 1>}.RGB) {g=true, r=true})
  n34 = ComponentMask(None=VectorParameter() {parameterName=Map2Offset, defaultValue=<rgba 0.7 0.7 0 1>}.RGB) {g=true, r=true}
  n35 = TextureSample(UVs=Add(A=@n28, B=Add(A=@n33, B=@n34)))
  n36 = Add(A=@n35.R, B=@n35.R)
  Multiply(
   A=Power(Base=Saturate(None=@n36), Exp=ScalarParameter() {defaultValue=0.8, group=Icons, parameterName=Map2 pow})
   B=ScalarParameter() {defaultValue=3, group=Icons, parameterName=Map2 Intensity}
  )
  n37 = TextureCoordinate()
  n38 = ComponentMask(None=VectorParameter() {parameterName=TexCoordScale, defaultValue=<rgba 4 4 0 0>}.RGB) {g=true, r=true}
  n39 = VectorParameter() {parameterName=TexCoordOffset, defaultValue=<rgba 0.0524 -1.7822 0 0>}
  NamedRerouteDeclaration(None=Add(A=Multiply(A=@n37, B=@n38), B=ComponentMask(None=@n39.RGB) {g=true, r=true})) {name=TexCoord, nodeColor=<rgba 1 0 0.4453 1>}
  TextureCoordinate() {vTiling=15}
  TextureCoordinate()
 ]
}
