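# Master subsurface material for the Terrain Man player skin, layering water, Zero Point and clock effects over standard character shading.
# TerrainManMask drives the layers: R optional ice, G water with scrolling foam normals and Fresnel colors, A switchable black, glossy Zero Point that pulls vertices to the object origin.
# The clock rotates a small UV disc in hourly steps from CurrentTime.
# Static switches gate glitch, panning emissive, clear coat and backpack water mapping.
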Material /BRCosmetics/Characters/Player/Male/Medium/Bodies/M_MED_Terrain_Man/Materials/M_MED_Terrain_Man_Master
graph {
 properties {
  bUseMaterialAttributes=true
  bUsedWithMeshDeformer=true
  bUsedWithMorphTargets=true
  bUsedWithSkeletalMesh=true
  pixelDepthOffsetMode=PDOM_Legacy
  shadingModel=MSM_Subsurface
  subsurfaceProfile=&SS_HeroSkin_02
  usedShadingModels=Subsurface
 }
 nodes [
  n1 = Multiply(A=Time(), B=ScalarParameter() {defaultValue=-0.1981, group=Water, parameterName=Water TimeScale})
  Frac(None=@n1)
  n2 = GetMaterialAttributes(None=NamedRerouteUsage()) {attributeGetTypes [#1, #2, #3]}
  n3 = Constant3Vector() {constant=<rgba 0.2103 0.224 0.2103 1>}
  n4 = ComponentMask(None=NamedRerouteUsage()) {g=true, r=true}
  n5 = TextureCoordinate()
  n6 = SphereMask(A=Distance(A=@n4, B=@n5), B=Constant() {r=0.026}, Radius=Constant() {r=0.003}, Hardness=Constant() {r=1})
  LinearInterpolate(A=@n2.Base Color, B=@n3, Alpha=@n6)
  n7 = Constant3Vector() {constant=<rgba 0.4844 0.4844 0.4844 1>}
  LinearInterpolate(A=@n2.Roughness, B=ComponentMask(None=@n7) {r=true}, Alpha=@n6)
  n8 = Multiply(A=NamedRerouteUsage(), B=ComponentMask(None=VectorParameter() {parameterName=zeropoint, defaultValue=<rgba 1 5 0 0>}.RGB) {g=true, r=true})
  n9 = Frac(None=Add(A=Constant2Vector() {r=0.5}, B=@n8))
  n10 = Frac(None=@n8)
  n11 = ComponentMask(None=@n10) {r=true}
  n12 = Multiply(A=Power(Base=Sine(None=@n11) {period=2})) {constB=5}
  LinearInterpolate(
   A=VectorParameter() {parameterName=zpBaseEmissive, defaultValue=<rgba 0.0219 0.1025 0.224 0>}.RGB
   B=LinearInterpolate(
    A=MaterialFunctionCall!(In=@n9) {materialFunction=/Game/Characters/Player/Female/Medium/Bodies/F_MED_Tower_Sentinel/Materials/MF_TowerSentinelChestEmissive}.Result
    B=MaterialFunctionCall(In=@n10) {materialFunction=&MF_TowerSentinelChestEmissive}.Result
    Alpha=Saturate(None=@n12)
   )
   Alpha=MaterialFunctionCall(Invert Fresnel=StaticBool() {value=true}, Power=ScalarParameter() {defaultValue=8, parameterName=zpFresnelEx}) {materialFunction=&Fresnel_Function}.Result
  )
  n13 = MaterialFunctionCall(TexCoords=NamedRerouteUsage()) {materialFunction=&BaseFN_CharacterFunction_PreFX}
  n14 = MaterialFunctionCall!(In=TextureObjectParameter() {group=Skin FX, parameterName=SkinFX_Mask}) {materialFunction=/Game/Weapons/WeaponSkins/MaterialFunction/MF_GlitchSubfunction}
  n15 = StaticSwitchParameter(True=@n14.Glitched, False=@n14.Glitched) {group=Skin FX, parameterName=UseGlitch}
  n16 = StaticSwitchParameter(
   True=MaterialFunctionCall!(
    Alpha=StaticComponentMaskParameter(None=@n15) {defaultR=true, group=Skin FX, parameterName=PanningEmissive_Channel}
    isVehicle
    Gmap=@n13.Gmap
    UVOffset=StaticComponentMaskParameter(None=@n15) {defaultB=true, group=Skin FX, parameterName=PanningEmissive_UVOffset_Channel}
    Local Normal
    Local Position
    EmissiveColorOverride
    Material=@n13.Gmap
   ) {materialFunction=/Game/Weapons/WeaponSkins/MaterialFunction/PanningEmissive}.Result
   False=@n13.Result
  ) {group=Skin FX, parameterName=Use Panning Emissive}
  n17 = StaticSwitchParameter(
   True=MaterialFunctionCall!(Alpha=StaticComponentMaskParameter(None=@n15) {defaultR=true, group=Skin FX, parameterName=Clear Coat Channel}, In=@n16) {materialFunction=/Game/Weapons/WeaponSkins/MaterialFunction/Clearcoat}.Result
   False=@n16
  ) {group=Skin FX, parameterName=Use Clear Coat}
  n18 = TextureSampleParameter2D() {
   parameterName=TerrainManMask
   samplerSource=SSM_Clamp_WorldGroupSettings
   samplerType=SAMPLERTYPE_Masks
   texture=/BRCosmetics/Characters/Player/Male/Medium/Bodies/M_MED_Terrain_Man/Textures/T_M_MED_Terrain_Man_Body_MasksForTechArt
  }
  n19 = MaterialFunctionCall!(Material=@n17, EdgeMap=Constant(), Alpha=@n18.R) {materialFunction=/Game/Weapons/WeaponSkins/MaterialFunction/Ice}
  n20 = GetMaterialAttributes(None=@n13.Result) {attributeGetTypes [#1]}
  n21 = Constant()
  n22 = BlendMaterialAttributes(
   A=StaticSwitchParameter(
    True=SetMaterialAttributes!(MaterialAttributes=@n19.Result, Base Color=@n20.Base Color, Subsurface Color=@n21) {attributeSetTypes [#1, #4]}
    False=@n17
   ) {group=Skin FX, parameterName=Use Ice}
   B=NamedRerouteUsage()
   Alpha=@n18.G
  )
  n23 = StaticSwitchParameter(True=BlendMaterialAttributes(A=@n22, B=NamedRerouteUsage(), Alpha=@n18.A), False=@n22) {defaultValue=true, parameterName=UseZeroPoint}
  MaterialFunctionCall(
   AO=@n13.AO
   Distance Blend=@n13.AO
   Fresnel=@n13.AO
   Gmap=@n13.AO
   Pre Skinned Local Position=@n13.AO
   Material=StaticSwitchParameter(True=NamedRerouteUsage(), False=@n23) {defaultValue=true, parameterName=UseClockAnimation}
  ) {materialFunction=&BaseFN_CharacterFunction_PostFX}
  NamedRerouteDeclaration(None=@n13.AO) {name=AO, nodeColor=<rgba 0 1 0.2969 1>}
  n24 = ComponentMask(None=VertexInterpolator(VS=PreSkinnedNormal()).PS) {g=true}
  NamedRerouteDeclaration(None=Saturate(None=Sign(None=@n24))) {name=FrontBackMask, nodeColor=<rgba 0 0.7891 1 1>}
  n25 = VectorParameter() {group=Water, parameterName=Water Edge Color, defaultValue=<rgba 0 0.1497 0.3281 1>}
  n26 = VectorParameter() {group=Water, parameterName=Water Core Color, defaultValue=<rgba 0.1489 0.4948 0.3899 1>}
  n27 = ScalarParameter() {defaultValue=2, group=Water, parameterName=Water exp}
  n28 = VectorParameter() {group=Water, parameterName=Water NormalTiling, defaultValue=<rgba 1 4 0 1>}
  n29 = AppendVector(A=@n28.R, B=@n28.R)
  n30 = VectorParameter() {group=Water, parameterName=Water NormalTilingBack, defaultValue=<rgba 1 4 0 1>}
  n31 = LinearInterpolate(A=AppendVector(A=@n30.R, B=@n30.R), B=@n29, Alpha=NamedRerouteUsage())
  n32 = Divide(A=PreSkinnedPosition(), B=PreSkinnedLocalBounds().Extents)
  n33 = ComponentMask(None=@n32) {b=true}
  n34 = SphereMask(
   A=@n33
   B=ScalarParameter() {defaultValue=0.815, group=Water, parameterName=p1}
   Radius=ScalarParameter() {defaultValue=0.2, group=Water, parameterName=r1}
   Hardness=ScalarParameter() {defaultValue=0.75, group=Water, parameterName=h1}
  )
  n35 = ScalarParameter() {defaultValue=0.2, group=Water, parameterName=i1}
  n36 = ScalarParameter() {defaultValue=0.6186, group=Water, parameterName=p2}
  n37 = ScalarParameter() {defaultValue=0.26, group=Water, parameterName=r2}
  n38 = ScalarParameter() {defaultValue=0.5, group=Water, parameterName=h2}
  n39 = Add(
   A=Add(A=ComponentMask(None=@n32) {r=true}, B=Multiply(A=@n34, B=@n35))
   B=Multiply(A=SphereMask(A=@n33, B=@n36, Radius=@n37, Hardness=@n38), B=ScalarParameter() {defaultValue=-0.5, group=Water, parameterName=i2})
  )
  n40 = NamedRerouteUsage()
  n41 = ComponentMask(None=@n40) {g=true}
  n42 = Saturate(None=LinearInterpolate(
    A=ScalarParameter() {defaultValue=-5, parameterName=WaterSpiralMin}
    B=ScalarParameter() {defaultValue=0.45, parameterName=WaterSpiralMax}
    Alpha=@n41
   ))
  n43 = AppendVector(
   A=Add(A=ComponentMask(None=@n40) {r=true}, B=Multiply(A=@n41, B=@n42))
   B=LinearInterpolate(
    A=ScalarParameter() {defaultValue=-5, parameterName=WaterSpiralDistMin}
    B=ScalarParameter() {defaultValue=0.45, parameterName=WaterSpiralDistMax}
    Alpha=@n41
   )
  )
  n44 = ComponentMask(None=VectorParameter() {group=RadialUV, parameterName=RadialUV_Mul, defaultValue=<rgba 1 1 0 0>}.RGB) {g=true, r=true}
  n45 = LinearInterpolate(A=VertexInterpolator(VS=AppendVector(A=@n39, B=@n33)).PS, B=Frac(None=Multiply(A=@n43, B=@n44)), Alpha=NamedRerouteUsage())
  n46 = StaticSwitchParameter(True=Multiply(A=TextureCoordinate(), B=@n29), False=ComponentMask(None=Multiply(A=@n31, B=@n45)) {g=true, r=true}) {parameterName=IsBackPack}
  n47 = ScalarParameter() {defaultValue=2, parameterName=WaterMipLevel}
  n48 = TextureSample!(UVs=@n46, Tex, Level=@n47, Apply View MipBias) {
   constMipValue=2
   mipValueMode=TMVM_MipLevel
   samplerSource=SSM_Wrap_WorldGroupSettings
   samplerType=SAMPLERTYPE_Normal
   texture=/Game/Athena/Apollo/Water/Textures/02_Rounded/02_WaterFoam_Veroni_base_N
  }
  n49 = ScalarParameter() {defaultValue=-0.2, group=Water, parameterName=Water NormalX Scale Back}
  n50 = ScalarParameter() {defaultValue=-0.2, group=Water, parameterName=Water NormalX Scale}
  n51 = NamedRerouteUsage()
  n52 = ScalarParameter() {defaultValue=0.55, group=Water, parameterName=Water NormalY Scale Back}
  n53 = ScalarParameter() {defaultValue=0.55, group=Water, parameterName=Water NormalY Scale}
  n54 = NamedRerouteUsage()
  n55 = AppendVector(A=Multiply(A=LinearInterpolate(A=@n49, B=@n50, Alpha=@n51), B=@n1), B=Multiply(A=LinearInterpolate(A=@n52, B=@n53, Alpha=@n54), B=@n1))
  n56 = TextureSample(UVs=Add(A=@n46, B=Frac(None=@n55)), Level=@n47) {
   constMipValue=2
   mipValueMode=TMVM_MipLevel
   samplerSource=SSM_Wrap_WorldGroupSettings
   samplerType=SAMPLERTYPE_Normal
   texture=&02_WaterFoam_Veroni_base_N
  }
  n57 = ScalarParameter() {defaultValue=-2, group=Water, parameterName=AO-ShoreMaskMin}
  n58 = ScalarParameter() {defaultValue=1, group=Water, parameterName=AO-ShoreMaskMax}
  n59 = NamedRerouteUsage()
  n60 = NamedRerouteUsage()
  n61 = LinearInterpolate(A=ScalarParameter() {defaultValue=-2, group=Water, parameterName=AO-ShoreMaskMinBack}, B=@n57, Alpha=@n60)
  n62 = LinearInterpolate(A=ScalarParameter() {defaultValue=1, group=Water, parameterName=AO-ShoreMaskMaxBack}, B=@n58, Alpha=@n60)
  n63 = Saturate(None=StaticSwitchParameter(True=LinearInterpolate(A=@n57, B=@n58, Alpha=@n59), False=LinearInterpolate(A=@n61, B=@n62, Alpha=@n59)) {parameterName=IsBackPack})
  n64 = ShadingPathSwitch(Default=LinearInterpolate(A=@n48.RGB, B=@n56.RGB, Alpha=@n63), Forward=@n56.RGB, Mobile=@n56.RGB)
  n65 = Transform(None=@n64)
  n66 = Saturate(None=Multiply(A=Fresnel(ExponentIn=@n27, Normal=@n65) {baseReflectFraction=0, exponent=0.25}))
  n67 = LinearInterpolate(
   A=VectorParameter() {group=Water, parameterName=Water Foam Color, defaultValue=<rgba 0.1901 0.4907 0.849 1>}.RGB
   B=Multiply(
    A=LinearInterpolate(A=@n25.RGB, B=@n26.RGB, Alpha=@n66)
    B=ScalarParameter() {defaultValue=0.8, group=Water, parameterName=Water SubsurfaceRatio}
   )
   Alpha=ComponentMask(None=@n63) {r=true}
  )
  n68 = Constant()
  n69 = Constant()
  n70 = SetMaterialAttributes!(
   MaterialAttributes
   Base Color=StaticSwitchParameter(True=@n67, False=ShadingPathSwitch(Default=@n68, Forward=@n67, Mobile=@n67)) {parameterName=RenderWaterToBaseColor}
   Normal=MaterialFunctionCall(Normal=@n64, Flatness=ScalarParameter() {group=Water, parameterName=WaterFlatness}) {materialFunction=&FlattenNormal}.Result
   Subsurface Color=StaticSwitchParameter(True=@n69, False=ShadingPathSwitch(Default=@n67, Forward=@n69, Mobile=@n69)) {parameterName=RenderWaterToBaseColor}
   Metallic=ScalarParameter() {defaultValue=0.1, group=Water, parameterName=WaterMetallic}
   Specular=ScalarParameter() {defaultValue=1, group=Water, parameterName=WaterSpec}
   Roughness=ScalarParameter() {defaultValue=0.15, group=Water, parameterName=WaterRoughness}
  ) {attributeSetTypes [
    #1
    #3
    #4
    #5
    #6
    #2
   ]}
  NamedRerouteDeclaration(None=@n70) {name=WaterAttributes, nodeColor=<rgba 0.125 1 0 1>}
  n71 = TextureCoordinate()
  n72 = ComponentMask(None=NamedRerouteDeclaration(None=VectorParameter() {parameterName=uvcenter, defaultValue=<rgba 0.483 0.443 0 0>}.RGB) {name=ClockCenter, nodeColor=<rgba 1 0 0.6328 1>}) {g=true, r=true}
  n73 = ScalarParameter() {defaultValue=6, parameterName=CurrentTime}
  n74 = OneMinus(None=Divide(A=Floor(None=@n73)) {constB=12})
  n75 = SphereMask(A=@n71, B=@n72, Radius=Constant() {r=0.0264}, Hardness=Constant() {r=1})
  NamedRerouteDeclaration(None=LinearInterpolate(
    A=@n71
    B=MaterialFunctionCall(UVs=@n71, Rotation Center=@n72, "Rotation Angle (0-1)"=@n74) {materialFunction=&CustomRotator}.Rotated Values
    Alpha=Floor(None=@n75)
   )) {name=ClockUVs, nodeColor=<rgba 0 1 0.6484 1>}
  NamedRerouteDeclaration(None=@n23) {name=ClockAnimIn, nodeColor=<rgba 0 0.6484 1 1>}
  n76 = PreSkinnedPosition()
  n77 = Subtract(
   A=VectorParameter() {group=RadialUV, parameterName=RadialUVOffset2, defaultValue=<rgba 0 0 130 0>}.RGB
   B=ShaderStageSwitch(PixelShader=VertexInterpolator(VS=@n76).PS, VertexShader=@n76)
  )
  n78 = ComponentMask(None=@n77) {b=true, r=true}
  n79 = Constant()
  n80 = PreSkinnedLocalBounds()
  n81 = Constant()
  n82 = OneMinus(None=Divide(A=Distance(A=@n78, B=@n79), B=Distance(A=@n80.Half Extents, B=@n81)))
  NamedRerouteDeclaration(None=@n82) {name=RadiusGradient, nodeColor=<rgba 0.2656 1 0 1>}
  n83 = Divide(A=@n77, B=@n80.Extents)
  n84 = ComponentMask(None=@n83) {b=true}
  n85 = ComponentMask(None=@n83) {r=true}
  n86 = Add(A=Divide(A=Arctangent2Fast(Y=@n84, X=@n85), B=Constant() {r=3.1416})) {constB=-0.5}
  NamedRerouteDeclaration(None=AppendVector(A=Multiply(A=@n86) {constB=0.5}, B=@n82)) {name=RadialUVs, nodeColor=<rgba 1 0.8438 0 1>}
  n87 = Constant()
  n88 = Subtract(A=ObjectPositionWS(), B=WorldPosition().XYZ)
  n89 = SetMaterialAttributes!(
   MaterialAttributes
   Emissive Color=@n87
   Base Color=@n87
   Subsurface Color=@n87
   Specular=Constant() {r=1}
   Roughness=Constant() {r=0.1}
   Metallic=@n87
   World Position Offset=@n88
  ) {attributeSetTypes [
    #7
    #1
    #4
    #6
    #2
    #5
    #8
   ]}
  NamedRerouteDeclaration(None=@n89) {name=ZeroPoint, nodeColor=<rgba 0 0.7891 1 1>}
  n90 = NamedRerouteUsage()
  n91 = SetMaterialAttributes!(MaterialAttributes=@n90, Normal=Constant3Vector() {constant=<rgba 0 0 1 0>}, Base Color=@n3, Roughness=@n7, Metallic=Constant()) {attributeSetTypes [
    #3
    #1
    #2
    #5
   ]}
  n92 = SphereMask(A=@n5, B=@n4, Radius=Constant() {r=0.033}, Hardness=Constant() {r=0.8})
  n93 = LinearInterpolate(
   A=Constant() {r=-0.5}
   B=Constant() {r=10}
   Alpha=TextureSampleParameter2D(UVs=NamedRerouteUsage()) {group=Diffuse, parameterName=Diffuse, texture=/Engine/EngineMaterials/DefaultDiffuse}.G
  )
  NamedRerouteDeclaration(None=BlendMaterialAttributes(A=@n90, B=@n91, Alpha=Min(A=@n92, B=@n93))) {name=ClockAnimOut, nodeColor=<rgba 1 0.1406 0 1>}
  ShadowReplace(Default=Constant(), Shadow=@n88)
 ]
}
