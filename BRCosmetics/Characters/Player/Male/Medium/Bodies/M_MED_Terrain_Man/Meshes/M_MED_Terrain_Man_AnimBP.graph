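# Post-process style cosmetic AnimBP for the Terrain Man body that corrects the right shoulder piece during weapon poses and arm movement.
# It copies the parent mesh pose, constrains Dyn_Clock_r to weighted clavicle and upper-arm bones, then stacks three interpolated additive offsets on Dyn_ShoulderPiece_r.
# The event graph drives those offsets from rifle/shotgun weapon type, pistol targeting, and a quaternion-twist alpha between right clavicle and upper arm.
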AnimBlueprint /BRCosmetics/Characters/Player/Male/Medium/Bodies/M_MED_Terrain_Man/Meshes/M_MED_Terrain_Man_AnimBP
target_skeleton=&/M_MED_Terrain_Man_Skeleton
graphs [
 AnimGraph {nodes [
   n1 = "Transform (Modify) Bone - Bone: Dyn_ShoulderPiece_r"(
    ComponentPose=Local To Component(LocalPose=Use cached pose 'Base_Pose'())
    Alpha="To Float (Boolean)"(InBool=Get Shotgun_AR_ADS() -> (Shotgun_AR_ADS:bool))
    Translation="0,0,1"
    Rotation="0,0,0"
    Scale="1,1,1"
   ) {
    rotationMode=BMM_Additive
    rotationSpace=BCS_BoneSpace
    translationMode=BMM_Additive
    translationSpace=BCS_ParentBoneSpace
    alphaBoolBlend {blendOption=Linear}
    alphaScaleBiasClamp {bInterpResult=true, interpSpeedDecreasing=8, interpSpeedIncreasing=20}
    boneToModify {boneName=Dyn_ShoulderPiece_r}
    translation {z=1}
   }
   n2 = "Transform (Modify) Bone - Bone: Dyn_ShoulderPiece_r"(
    ComponentPose=@n1
    Alpha="To Float (Boolean)"(InBool=Get Pistol_ADS() -> (Pistol_ADS:bool))
    Translation="1,0.5,0"
    Rotation="0,0,0"
    Scale="1,1,1"
   ) {
    rotationMode=BMM_Additive
    rotationSpace=BCS_BoneSpace
    translationMode=BMM_Additive
    translationSpace=BCS_ParentBoneSpace
    alphaBoolBlend {blendOption=Linear}
    alphaScaleBiasClamp {bInterpResult=true, interpSpeedDecreasing=20, interpSpeedIncreasing=20}
    boneToModify {boneName=Dyn_ShoulderPiece_r}
    translation=<1 0.5>
   }
   n3 = "Transform (Modify) Bone - Bone: Dyn_ShoulderPiece_r"(ComponentPose=@n2, Alpha=Get Arm_Rotation() -> (Arm_Rotation:double), Translation="3,1,2", Rotation="0,0,0", Scale="1,1,1") {
    rotationMode=BMM_Additive
    rotationSpace=BCS_BoneSpace
    translationMode=BMM_Additive
    translationSpace=BCS_ParentBoneSpace
    alphaBoolBlend {blendOption=Linear}
    alphaScaleBiasClamp {bInterpResult=true, interpSpeedDecreasing=20, interpSpeedIncreasing=20}
    boneToModify {boneName=Dyn_ShoulderPiece_r}
    translation=<3 1 2>
   }
   Output Pose(Result=Component To Local(ComponentPose=@n3)) {layerGroup=None}
   n4 = "Constraint - Dyn_Clock_r (PPP)"!(
    ComponentPose=Local To Component(LocalPose=Copy Pose From Mesh() {bUseAttachedParent=true})
    Alpha=1
    ConstraintWeights_0=0.1
    ConstraintWeights_1:float=0.1
    ConstraintWeights_2:float=0.8
   ) -> (Pose) {
    alphaBoolBlend {blendOption=Linear}
    boneToModify {boneName=Dyn_Clock_r}
    constraintSetup [{targetBone {boneName=clavicle_r}}, {targetBone {boneName=upperarm_r}}, {targetBone {boneName=upperarm_twist_01_r}}]
    constraintWeights [1,1,1]
   }
   Save cached pose 'Base_Pose'(Pose=Component To Local(ComponentPose=@n4)) {cacheName=Base_Pose}
  ]}
 EventGraph {nodes [
   Event BlueprintInitializeAnimation()
   > Set Skel_Mesh(Skel_Mesh=GetOwnerSkeletalMeshComponent())
   Event BlueprintUpdateAnimation()
   > n1 = Sequence()
   > n2 = Cast To FortPlayerAnimInstance(Object=GetAnimInstance(self=GetOwnerSkeletalMeshComponent()))
   > Set Pistol_ADS(Pistol_ADS:bool=false) -> (Output_Get:bool)
   > Set Shotgun_AR_ADS(Shotgun_AR_ADS:bool=false) -> (Output_Get:bool)
   > n3 = Switch on EFortWeaponCoreAnimation!(execute, Selection:EFortWeaponCoreAnimation Enum=Get WeaponCoreAnim(self=@n2)) -> (
    Melee:Exec
    Pistol:Exec
    Shotgun:Exec
    PaperBlueprint:Exec
    Rifle:Exec
    MeleeOneHand:Exec
    MachinePistol:Exec
    RocketLauncher:Exec
    GrenadeLauncher:Exec
    GoingCommando:Exec
    AssaultRifle:Exec
    TacticalShotgun:Exec
    SniperRifle:Exec
    TrapPlacement:Exec
    ShoulderLauncher:Exec
    AbilityDecoTool:Exec
    Crossbow:Exec
    C4:Exec
    RemoteControl:Exec
    DualWield:Exec
    AR_BullPup:Exec
    AR_ForwardGrip:Exec
    MedPackPaddles:Exec
    SMG_P90:Exec
    AR_DrumGun:Exec
    Consumable_Small:Exec
    Consumable_Large:Exec
    Balloon:Exec
    MountedTurret:Exec
    CreativeTool:Exec
    ExplosiveBow:Exec
    AshtonIndigo:Exec
    AshtonChicago:Exec
    MeleeDualWield:Exec
    Unarmed:Exec
   )
   Branch(execute=[@n3.AssaultRifle, @n3.Rifle, @n3.Shotgun, @n3.TacticalShotgun])
   > Set Shotgun_AR_ADS(Shotgun_AR_ADS=true)
   Branch(execute=@n3.Pistol, Condition=Get bIsTargeting())
   > Set Pistol_ADS(Pistol_ADS=true)
   Set Arm_Rotation(
    execute=@n1.then_1
    Arm_Rotation:double=GetAlphaFromReferenceTransformDelta(
     SkeletalMeshComponent=Get Skel_Mesh()
     BoneName=clavicle_r
     BaseName=upperarm_r
     DeltaComponentType=QuaternionTwist
     InputRange=(X=67,Y=128.9)
     TwistAxis="0,0,1"
    )
   ) -> (Output_Get:double)
  ]}
]
variables [
 Aiming {type=boolean}
 AR {type=number}
 AR_Equip {type=number}
 Arm_Rotation {type=number}
 Mini_Gun {type=number}
 Pistol {type=number}
 Pistol_ADS {type=boolean}
 Pistol_Equip {type=number}
 Shotgun_AR_ADS {type=boolean}
 Shotgun_AR_ADS_Down {type=boolean}
 Skel_Mesh {type=SkeletalMeshComponent}
 weapon_aim_Down {type=number}
 weapon_aim_Up {type=number}
]
